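# Material function that draws a reactive Pip-Boy-style emissive screen on the RadiumFox backpack, over a blacked-out base color.
# ReactivityStatus selects the frame and gates walking, elimination, chest-opening, damage and healing displays.
# Custom UV remaps place flipbook emotes, a two-digit elimination counter and a time-driven healing bar.
# It also builds camera-distance scanlines with Fresnel, time-based damage distortion and a channel-selectable TechArtMask.
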MaterialFunction /BRCosmetics/Accessories/FORT_Backpacks/Backpack_M_MED_RadiumFox/Materials/MF_RadiumFox_Screen
graph {nodes [
  n1 = GetMaterialAttributes(None=NamedRerouteUsage()) {attributeGetTypes [#1, #2, #3]}
  n2 = NamedRerouteUsage()
  FunctionOutput(None=SetMaterialAttributes!(MaterialAttributes=@n1.MaterialAttributes, EmissiveColor=Add(A=@n2, B=@n1.EmissiveColor), BaseColor=Constant()) {attributeSetTypes [#1, #3]})
  NamedRerouteDeclaration(None=StaticSwitchParameter(True=TextureCoordinate() {coordinateIndex=1}, False=TextureCoordinate()) {defaultValue=true, group=Reactivity, parameterName=UseUVChannel01}) {name=UV, nodeColor=<rgba 0.3828 1 0 1>}
  NamedRerouteDeclaration(None=FunctionInput() {inputName=Material, inputType=FunctionInput_MaterialAttributes}) {name=MA, nodeColor=<rgba 0 1 0.9766 1>}
  n3 = Constant()
  NamedRerouteDeclaration(None=If(A=NamedRerouteUsage(), "A > B"=@n3, "A == B"=Constant() {r=1}, "A < B"=@n3) {constB=2}) {name=IsOpeningChest, nodeColor=<rgba 0.3828 0 1 1>}
  n4 = Multiply(A=ScalarParameter() {defaultValue=1, group=Emote, parameterName=OpenChestEmoteSpeed}, B=Time())
  n5 = Floor(None=LinearInterpolate(Alpha=Frac(None=@n4)) {constA=4, constB=8})
  n6 = Constant2Vector() {g=4, r=4}
  n7 = Multiply(A=@n6.R, B=@n6.R)
  n8 = VectorParameter() {group=Numbers, parameterName=OpenChest_Counter, channelNames {a=ScaleY, b=ScaleX, g=CenterY, r=CenterX}, defaultValue=<rgba 0.5 0.5 1 1>}
  MaterialExpressionAppendVector_0 = AppendVector(A=@n8.CenterX, B=@n8.CenterX)
  MaterialExpressionAppendVector_1 = AppendVector(A=@n8.ScaleX, B=@n8.ScaleX)
  MaterialExpressionNamedRerouteUsage_41 = NamedRerouteUsage()
  n9 = Custom!(UVPoint=@MaterialExpressionAppendVector_0, UVScale=@MaterialExpressionAppendVector_1, OriginalUV=@MaterialExpressionNamedRerouteUsage_41) {
   code="float2 DistUV =OriginalUV-UVPoint;\r\nDistUV = DistUV/UVScale + float2(0.5,0.5);\r\n\r\nreturn DistUV;"
   description=CopyToPointUV
   outputType=CMOT_Float2
   inputs [
    {
     inputName=UVPoint
     input {expression=&self:MaterialExpressionAppendVector_0}
    }
    {
     inputName=UVScale
     input {expression=&self:MaterialExpressionAppendVector_1}
    }
    {
     inputName=OriginalUV
     input {expression=&self:MaterialExpressionNamedRerouteUsage_41}
    }
   ]
  }
  n10 = Clamp(None=@n9)
  n11 = TextureSample(
   UVs=MaterialFunctionCall("Animation  Phase (0-1)"=Divide(A=@n5, B=@n7) {constB=4}, Number of Rows=@n6.R, Number of Columns=@n6.R, UVs=@n10) {materialFunction=&FlipBook}.UVs
   Tex=NamedRerouteUsage()
  ) {samplerSource=SSM_Wrap_WorldGroupSettings, texture=/BRCosmetics/Accessories/FORT_Backpacks/Backpack_M_MED_RadiumFox/Textures/T_PipBoyEmotes}
  n12 = MaterialFunctionCall!(UV=@n10) {materialFunction=&/MF_UVMask01Space}
  NamedRerouteDeclaration(None=Multiply(A=Multiply(A=@n11.RGB, B=@n12.Result), B=NamedRerouteUsage())) {name=ChestOpenEmote, nodeColor=<rgba 0 1 0.6953 1>}
  NamedRerouteDeclaration(None=VectorParameter() {group=Reactivity, parameterName=DarkGreenColor, defaultValue=<rgba 1 1 1 1>}.RGB) {name=DarkGreen, nodeColor=<rgba 1 0.8906 0 1>}
  n13 = ScalarParameter() {group=Reactivity, parameterName=ReactivityStatus}
  NamedRerouteDeclaration(None=@n13) {name=Status, nodeColor=<rgba 0.2422 0 1 1>}
  n14 = Constant()
  NamedRerouteDeclaration(None=If(A=NamedRerouteUsage(), "A > B"=@n14, "A == B"=Constant() {r=1}, "A < B"=@n14) {constB=1}) {name=IsElimation, nodeColor=<rgba 1 0 0.5625 1>}
  n15 = Constant()
  NamedRerouteDeclaration(None=If(A=NamedRerouteUsage(), "A > B"=@n15, "A == B"=Constant() {r=1}, "A < B"=@n15) {constB=3}) {name=IsReceivingDamage, nodeColor=<rgba 0.3828 0 1 1>}
  n16 = NamedRerouteUsage()
  n17 = ScalarParameter() {defaultValue=2, group=Reactivity, parameterName=ScanlineTilling}
  n18 = NamedRerouteUsage()
  n19 = NamedRerouteUsage()
  n20 = TextureSampleParameter2D(UVs=Add(A=Multiply(A=@n16, B=@n17), B=Multiply(A=@n18, B=@n19))) {group=Reactivity, parameterName=ScanLineMask, texture=/BRCosmetics/Accessories/FORT_Backpacks/Backpack_M_MED_RadiumFox/Textures/T_ScanLines}
  n21 = CameraPositionWS()
  n22 = ActorPositionWS()
  n23 = Divide(
   A=Length(None=Subtract(A=@n21, B=@n22)) {is=class MaterialExpressionLength}
   B=ScalarParameter() {defaultValue=300, parameterName=ScanLine_Gradient}
  )
  n24 = Add(
   A=Multiply(A=ScalarParameter() {defaultValue=0.3, parameterName=RoughnessAffect}, B=NamedRerouteUsage())
   B=LinearInterpolate(
    A=ScalarParameter() {defaultValue=0.1, group=Reactivity, parameterName=ScanLineIntensity_Near}
    B=ScalarParameter() {defaultValue=0.7, group=Reactivity, parameterName=ScanLineIntensity_Far}
    Alpha=Saturate(None=@n23)
   )
  )
  n25 = NamedRerouteUsage()
  n26 = Panner(
   Coordinate=Multiply(A=NamedRerouteUsage(), B=ScalarParameter() {defaultValue=1.3, group=Reactivity, parameterName=DynamicScanlineTilling})
   Speed=AppendVector(
    A=ScalarParameter() {group=Reactivity, parameterName=ScanSpeedX}
    B=ScalarParameter() {defaultValue=-0.2, group=Reactivity, parameterName=ScanSpeedY}
   )
  )
  n27 = Power(Base=Frac(None=ComponentMask(None=@n26) {g=true})) {constExponent=20}
  n28 = Add(
   A=Multiply(A=@n20.R, B=Saturate(None=@n24))
   B=Multiply(A=ScalarParameter() {defaultValue=0.5, group=Reactivity, parameterName=DynamicLineIntensity}, B=Multiply(A=@n25, B=@n27))
  )
  n29 = NamedRerouteUsage()
  n30 = NamedRerouteUsage()
  n31 = Fresnel()
  NamedRerouteDeclaration(None=Add(A=Multiply(A=@n28, B=@n29), B=Multiply(A=@n30, B=@n31))) {name=ScanLine, nodeColor=<rgba 0 1 0.6719 1>}
  n32 = TextureObjectParameter() {group=Emote, parameterName=EmoteAtlas, texture=&T_PipBoyEmotes}
  NamedRerouteDeclaration(None=@n32) {name=EmoteTex, nodeColor=<rgba 0.4766 0 1 1>}
  n33 = NamedRerouteUsage()
  n34 = NamedRerouteUsage()
  n35 = NamedRerouteUsage()
  n36 = Constant2Vector() {g=2, r=2}
  n37 = Divide(A=@n13, B=Multiply(A=@n36.R, B=@n36.R)) {constB=4}
  n38 = Subtract(A=Divide() {constA=3, constB=4}) {constB=0.01}
  n39 = VectorParameter() {group=Reactivity, parameterName=Interface_Counter, channelNames {a=ScaleY, b=ScaleX, g=CenterY, r=CenterX}, defaultValue=<rgba 0.5 0.5 1 1>}
  MaterialExpressionAppendVector_3 = AppendVector(A=@n39.CenterX, B=@n39.CenterX)
  MaterialExpressionAppendVector_4 = AppendVector(A=@n39.ScaleX, B=@n39.ScaleX)
  MaterialExpressionNamedRerouteUsage_2 = NamedRerouteUsage()
  n40 = Custom(UVPoint=@MaterialExpressionAppendVector_3, UVScale=@MaterialExpressionAppendVector_4, OriginalUV=@MaterialExpressionNamedRerouteUsage_2) {
   code="float2 DistUV =OriginalUV-UVPoint;\r\nDistUV = DistUV/UVScale + float2(0.5,0.5);\r\n\r\nreturn DistUV;"
   description=CopyToPointUV
   outputType=CMOT_Float2
   inputs [
    {
     inputName=UVPoint
     input {expression=&self:MaterialExpressionAppendVector_3}
    }
    {
     inputName=UVScale
     input {expression=&self:MaterialExpressionAppendVector_4}
    }
    {
     inputName=OriginalUV
     input {expression=&self:MaterialExpressionNamedRerouteUsage_2}
    }
   ]
  }
  n41 = TextureSampleParameter2D(UVs=MaterialFunctionCall("Animation  Phase (0-1)"=Clamp(None=@n37, Max=@n38) {maxDefault=0.99}, Number of Rows=@n36.R, Number of Columns=@n36.R, UVs=@n40) {materialFunction=&FlipBook}.UVs) {
   group=Reactivity
   parameterName=ScreenFrameWork
   texture=/BRCosmetics/Accessories/FORT_Backpacks/Backpack_M_MED_RadiumFox/Textures/T_PipBoyFrameWork
  }
  n42 = Multiply(
   A=@n41.RGB
   B=MaterialFunctionCall(UV=@n40) {materialFunction=&MF_UVMask01Space}.Result
  )
  n43 = Add(A=NamedRerouteUsage(), B=Add(A=NamedRerouteUsage(), B=Add(A=@n35, B=@n42)))
  n44 = Add(A=NamedRerouteUsage(), B=Add(A=NamedRerouteUsage(), B=Add(A=@n34, B=@n43)))
  NamedRerouteDeclaration(None=Add(A=Multiply(A=@n33, B=@n44), B=NamedRerouteUsage())) {name=EM_OUT, nodeColor=<rgba 1 0.6563 0 1>}
  n45 = NamedRerouteUsage()
  n46 = Multiply(A=ScalarParameter() {defaultValue=1, group=Emote, parameterName=ElimationEmoteSpeed}, B=Time())
  n47 = Floor(None=LinearInterpolate(Alpha=Frac(None=@n46)) {constB=4})
  n48 = Constant2Vector() {g=4, r=4}
  n49 = Multiply(A=@n48.R, B=@n48.R)
  n50 = VectorParameter() {group=Numbers, parameterName=Elimation_Counter, channelNames {a=ScaleY, b=ScaleX, g=CenterY, r=CenterX}, defaultValue=<rgba 0.5 0.5 1 1>}
  MaterialExpressionAppendVector_7 = AppendVector(A=@n50.CenterX, B=@n50.CenterX)
  MaterialExpressionAppendVector_8 = AppendVector(A=@n50.ScaleX, B=@n50.ScaleX)
  MaterialExpressionNamedRerouteUsage_21 = NamedRerouteUsage()
  n51 = Custom(UVPoint=@MaterialExpressionAppendVector_7, UVScale=@MaterialExpressionAppendVector_8, OriginalUV=@MaterialExpressionNamedRerouteUsage_21) {
   code="float2 DistUV =OriginalUV-UVPoint;\r\nDistUV = DistUV/UVScale + float2(0.5,0.5);\r\n\r\nreturn DistUV;"
   description=CopyToPointUV
   outputType=CMOT_Float2
   inputs [
    {
     inputName=UVPoint
     input {expression=&self:MaterialExpressionAppendVector_7}
    }
    {
     inputName=UVScale
     input {expression=&self:MaterialExpressionAppendVector_8}
    }
    {
     inputName=OriginalUV
     input {expression=&self:MaterialExpressionNamedRerouteUsage_21}
    }
   ]
  }
  n52 = Clamp(None=@n51)
  n53 = TextureSample(
   UVs=MaterialFunctionCall("Animation  Phase (0-1)"=Divide(A=@n47, B=@n49) {constB=4}, Number of Rows=@n48.R, Number of Columns=@n48.R, UVs=@n52) {materialFunction=&FlipBook}.UVs
   Tex=NamedRerouteUsage()
  ) {samplerSource=SSM_Wrap_WorldGroupSettings, texture=&T_PipBoyEmotes}
  n54 = Multiply(
   A=@n53.RGB
   B=MaterialFunctionCall(UV=@n52) {materialFunction=&MF_UVMask01Space}.Result
  )
  NamedRerouteDeclaration(None=Multiply(A=Add(A=@n45, B=@n54), B=NamedRerouteUsage())) {name=ElimationEmote, nodeColor=<rgba 0.9688 1 0 1>}
  n55 = VectorParameter() {group=Numbers, parameterName=HealingBar_Center, channelNames {a=ScaleY, b=ScaleX, g=CenterY, r=CenterX}, defaultValue=<rgba 0.5 0.5 1 1>}
  MaterialExpressionAppendVector_19 = AppendVector(A=@n55.CenterX, B=@n55.CenterX)
  MaterialExpressionAppendVector_20 = AppendVector(A=@n55.ScaleX, B=@n55.ScaleX)
  MaterialExpressionNamedRerouteUsage_6 = NamedRerouteUsage()
  n56 = Custom(UVPoint=@MaterialExpressionAppendVector_19, UVScale=@MaterialExpressionAppendVector_20, OriginalUV=@MaterialExpressionNamedRerouteUsage_6) {
   code="float2 DistUV =OriginalUV-UVPoint;\r\nDistUV = DistUV/UVScale + float2(0.5,0.5);\r\n\r\nreturn DistUV;"
   description=CopyToPointUV
   outputType=CMOT_Float2
   inputs [
    {
     inputName=UVPoint
     input {expression=&self:MaterialExpressionAppendVector_19}
    }
    {
     inputName=UVScale
     input {expression=&self:MaterialExpressionAppendVector_20}
    }
    {
     inputName=OriginalUV
     input {expression=&self:MaterialExpressionNamedRerouteUsage_6}
    }
   ]
  }
  n57 = Clamp(None=@n56)
  n58 = TextureSampleParameter2D(UVs=@n57) {
   group=Emote
   parameterName=HealthProgressBar
   samplerType=SAMPLERTYPE_Masks
   texture=/BRCosmetics/Accessories/FORT_Backpacks/Backpack_M_MED_RadiumFox/Textures/T_PipBoyBar
  }
  n59 = NamedRerouteUsage()
  n60 = Divide(
   A=Subtract(A=Time(), B=ScalarParameter() {group=Emote, parameterName=HealingStartTime})
   B=ScalarParameter() {defaultValue=2, parameterName=HealingTime}
  )
  n61 = Step(
   Y=ComponentMask(None=NamedRerouteUsage()) {r=true}
   X=LinearInterpolate(
    A=ScalarParameter() {group=Reactivity, parameterName=LifeStartX}
    B=ScalarParameter() {defaultValue=1, group=Reactivity, parameterName=LifeEndX}
    Alpha=Frac(None=@n60)
   )
  )
  n62 = Add(A=Multiply(A=@n58.R, B=@n59), B=Multiply(A=@n59, B=Multiply(A=@n58.G, B=@n61)))
  n63 = MaterialFunctionCall(UV=@n57) {materialFunction=&MF_UVMask01Space}
  NamedRerouteDeclaration(None=Multiply(A=Multiply(A=@n62, B=@n63.Result), B=NamedRerouteUsage())) {name=HealingProgress, nodeColor=<rgba 0.875 0 1 1>}
  n64 = VectorParameter() {
   group=Numbers
   parameterName=ElimationNumber_Counter
   channelNames {a=ScaleY, b=ScaleX, g=CenterY, r=CenterX}
   defaultValue=<rgba 0.5 0.5 1 1>
  }
  MaterialExpressionAppendVector_9 = AppendVector(A=@n64.CenterX, B=@n64.CenterX)
  NamedRerouteDeclaration(None=@MaterialExpressionAppendVector_9) {name=NumberStart_Elimation, nodeColor=<rgba 0 1 0.6016 1>}
  MaterialExpressionAppendVector_10 = AppendVector(A=@n64.ScaleX, B=@n64.ScaleX)
  NamedRerouteDeclaration(None=@MaterialExpressionAppendVector_10) {name=NumberScale_Elimation, nodeColor=<rgba 1 0.2812 0 1>}
  n65 = ScalarParameter() {group=Numbers, parameterName=ElimationNumberInt}
  NamedRerouteDeclaration(None=@n65) {name=ElimationNumberInt, nodeColor=<rgba 0.9453 1 0 1>}
  n66 = Time()
  n67 = LinearInterpolate(Alpha=Frac(None=Multiply(A=@n66) {constB=0.5})) {constA=8, constB=12}
  n68 = Divide(
   A=Subtract(A=Time(), B=ScalarParameter() {group=Emote, parameterName=DamageReceivedStartTime})
   B=ScalarParameter() {defaultValue=2, group=Emote, parameterName=DamageReceivedEmoteTime}
  )
  n69 = ScalarParameter() {group=Emote, parameterName=DamageReceived_FrontEnd}
  n70 = LinearInterpolate(Alpha=LinearInterpolate(A=Saturate(None=@n68), B=Saturate(None=@n69), Alpha=ScalarParameter() {group=Emote, parameterName=IsFrontEnd})) {constA=8, constB=11}
  n71 = Constant2Vector() {g=4, r=4}
  n72 = Divide(
   A=Floor(None=StaticSwitchParameter(True=@n67, False=@n70) {group=Emote, parameterName=PreviewDamageReceived})
   B=Multiply(A=@n71.R, B=@n71.R)
  ) {constB=4}
  n73 = VectorParameter() {
   group=Numbers
   parameterName=DamageReceived_Counter
   channelNames {a=ScaleY, b=ScaleX, g=CenterY, r=CenterX}
   defaultValue=<rgba 0.5 0.5 1 1>
  }
  MaterialExpressionAppendVector_2 = AppendVector(A=@n73.CenterX, B=@n73.CenterX)
  MaterialExpressionAppendVector_12 = AppendVector(A=@n73.ScaleX, B=@n73.ScaleX)
  MaterialExpressionNamedRerouteUsage_46 = NamedRerouteUsage()
  n74 = Custom(UVPoint=@MaterialExpressionAppendVector_2, UVScale=@MaterialExpressionAppendVector_12, OriginalUV=@MaterialExpressionNamedRerouteUsage_46) {
   code="float2 DistUV =OriginalUV-UVPoint;\r\nDistUV = DistUV/UVScale + float2(0.5,0.5);\r\n\r\nreturn DistUV;"
   description=CopyToPointUV
   outputType=CMOT_Float2
   inputs [
    {
     inputName=UVPoint
     input {expression=&self:MaterialExpressionAppendVector_2}
    }
    {
     inputName=UVScale
     input {expression=&self:MaterialExpressionAppendVector_12}
    }
    {
     inputName=OriginalUV
     input {expression=&self:MaterialExpressionNamedRerouteUsage_46}
    }
   ]
  }
  n75 = Clamp(None=@n74)
  n76 = TextureSample(
   UVs=Add(
    A=NamedRerouteUsage()
    B=MaterialFunctionCall("Animation  Phase (0-1)"=@n72, Number of Rows=@n71.R, Number of Columns=@n71.R, UVs=@n75) {materialFunction=&FlipBook}.UVs
   )
   Tex=NamedRerouteUsage()
  ) {samplerSource=SSM_Wrap_WorldGroupSettings, texture=&T_PipBoyEmotes}
  n77 = NamedRerouteUsage()
  n78 = Clamp(None=@n74)
  n79 = Multiply(
   A=@n76.RGB
   B=MaterialFunctionCall(UV=Add(A=@n77, B=@n78)) {materialFunction=&MF_UVMask01Space}.Result
  )
  n80 = Multiply(A=NamedRerouteUsage(), B=NamedRerouteUsage())
  n81 = NamedRerouteUsage()
  n82 = Multiply(
   A=Multiply(A=Multiply(A=@n80, B=@n81), B=NamedRerouteUsage())
   B=ScalarParameter() {defaultValue=0.8, group=Emote, parameterName=BreakColorIntensity}
  )
  NamedRerouteDeclaration(None=Multiply(A=Add(A=@n79, B=@n82), B=NamedRerouteUsage())) {name=DamageReceivedEmote, nodeColor=<rgba 0 0.9297 1 1>}
  n83 = NamedRerouteUsage()
  n84 = Constant() {r=10}
  n85 = Clamp(None=Floor(None=Fmod(A=@n83, B=@n84))) {maxDefault=9}
  n86 = Constant2Vector() {g=4, r=4}
  n87 = Multiply(A=@n86.R, B=@n86.R)
  MaterialExpressionAdd_8 = Add(A=NamedRerouteUsage(), B=AppendVector(A=ScalarParameter() {group=Numbers, parameterName=NumberElimationOffsetX}, B=Constant()))
  MaterialExpressionNamedRerouteUsage_39 = NamedRerouteUsage()
  MaterialExpressionNamedRerouteUsage_34 = NamedRerouteUsage()
  n88 = Custom(UVPoint=@MaterialExpressionAdd_8, UVScale=@MaterialExpressionNamedRerouteUsage_39, OriginalUV=@MaterialExpressionNamedRerouteUsage_34) {
   code="float2 DistUV =OriginalUV-UVPoint;\r\nDistUV = DistUV/UVScale + float2(0.5,0.5);\r\n\r\nreturn DistUV;"
   description=CopyToPointUV
   outputType=CMOT_Float2
   inputs [
    {
     inputName=UVPoint
     input {expression=&self:MaterialExpressionAdd_8}
    }
    {
     inputName=UVScale
     input {expression=&self:MaterialExpressionNamedRerouteUsage_39}
    }
    {
     inputName=OriginalUV
     input {expression=&self:MaterialExpressionNamedRerouteUsage_34}
    }
   ]
  }
  n89 = Clamp(None=@n88)
  n90 = TextureSample(
   UVs=MaterialFunctionCall("Animation  Phase (0-1)"=Divide(A=@n85, B=@n87) {constB=4}, Number of Rows=@n86.R, Number of Columns=@n86.R, UVs=@n89) {materialFunction=&FlipBook}.UVs
   Tex=NamedRerouteUsage()
  ) {
   samplerSource=SSM_Wrap_WorldGroupSettings
   texture=/BRCosmetics/Accessories/FORT_Backpacks/Backpack_M_MED_RadiumFox/Textures/T_FallOutNumbers
  }
  n91 = Multiply(
   A=@n90.RGB
   B=MaterialFunctionCall(UV=@n89) {materialFunction=&MF_UVMask01Space}.Result
  )
  n92 = Divide(A=@n65) {constB=10}
  n93 = Clamp(None=Fmod(A=Floor(None=@n92), B=Constant() {r=10})) {maxDefault=9}
  n94 = Constant2Vector() {g=4, r=4}
  n95 = Multiply(A=@n94.R, B=@n94.R)
  MaterialExpressionNamedRerouteUsage_30 = NamedRerouteUsage()
  n96 = Custom(UVPoint=@MaterialExpressionAppendVector_9, UVScale=@MaterialExpressionAppendVector_10, OriginalUV=@MaterialExpressionNamedRerouteUsage_30) {
   code="float2 DistUV =OriginalUV-UVPoint;\r\nDistUV = DistUV/UVScale + float2(0.5,0.5);\r\n\r\nreturn DistUV;"
   description=CopyToPointUV
   outputType=CMOT_Float2
   inputs [
    {
     inputName=UVPoint
     input {expression=&self:MaterialExpressionAppendVector_9}
    }
    {
     inputName=UVScale
     input {expression=&self:MaterialExpressionAppendVector_10}
    }
    {
     inputName=OriginalUV
     input {expression=&self:MaterialExpressionNamedRerouteUsage_30}
    }
   ]
  }
  n97 = Clamp(None=@n96)
  n98 = TextureSample(
   UVs=MaterialFunctionCall("Animation  Phase (0-1)"=Divide(A=@n93, B=@n95) {constB=4}, Number of Rows=@n94.R, Number of Columns=@n94.R, UVs=@n97) {materialFunction=&FlipBook}.UVs
   Tex=NamedRerouteUsage()
  ) {
   samplerSource=SSM_Wrap_WorldGroupSettings
   texture=&T_FallOutNumbers
  }
  n99 = Multiply(
   A=@n98.RGB
   B=MaterialFunctionCall(UV=@n97) {materialFunction=&MF_UVMask01Space}.Result
  )
  NamedRerouteDeclaration(None=Multiply(A=Add(A=@n91, B=@n99), B=NamedRerouteUsage())) {name=ElimationNumber, nodeColor=<rgba 0.3125 1 0 1>}
  NamedRerouteDeclaration(None=@n20.R) {name=LineMask, nodeColor=<rgba 1 0.3516 0 1>}
  n100 = Constant()
  NamedRerouteDeclaration(None=If(A=NamedRerouteUsage(), "A > B"=@n100, "A == B"=Constant() {r=1}, "A < B"=@n100) {constB=4}) {name=IsHealing, nodeColor=<rgba 0.1016 0 1 1>}
  NamedRerouteDeclaration(None=@n1.Roughness) {name=Roughness_IN, nodeColor=<rgba 1 0 0.7969 1>}
  NamedRerouteDeclaration(None=VectorParameter() {group=Reactivity, parameterName=HealthBarColor, defaultValue=<rgba 1 1 1 1>}.RGB) {name=BarCd, nodeColor=<rgba 0.3359 0 1 1>}
  n101 = Multiply(A=ScalarParameter() {defaultValue=1, group=Emote, parameterName=HealingEmoteSpeed}, B=Time())
  n102 = Floor(None=LinearInterpolate(Alpha=Frac(None=@n101)) {constA=12, constB=16})
  n103 = Constant2Vector() {g=4, r=4}
  n104 = Multiply(A=@n103.R, B=@n103.R)
  n105 = VectorParameter() {group=Numbers, parameterName=Healing_Counter, channelNames {a=ScaleY, b=ScaleX, g=CenterY, r=CenterX}, defaultValue=<rgba 0.5 0.5 1 1>}
  MaterialExpressionAppendVector_13 = AppendVector(A=@n105.CenterX, B=@n105.CenterX)
  MaterialExpressionAppendVector_14 = AppendVector(A=@n105.ScaleX, B=@n105.ScaleX)
  MaterialExpressionNamedRerouteUsage_52 = NamedRerouteUsage()
  n106 = Custom(UVPoint=@MaterialExpressionAppendVector_13, UVScale=@MaterialExpressionAppendVector_14, OriginalUV=@MaterialExpressionNamedRerouteUsage_52) {
   code="float2 DistUV =OriginalUV-UVPoint;\r\nDistUV = DistUV/UVScale + float2(0.5,0.5);\r\n\r\nreturn DistUV;"
   description=CopyToPointUV
   outputType=CMOT_Float2
   inputs [
    {
     inputName=UVPoint
     input {expression=&self:MaterialExpressionAppendVector_13}
    }
    {
     inputName=UVScale
     input {expression=&self:MaterialExpressionAppendVector_14}
    }
    {
     inputName=OriginalUV
     input {expression=&self:MaterialExpressionNamedRerouteUsage_52}
    }
   ]
  }
  n107 = Clamp(None=@n106)
  n108 = TextureSample(
   UVs=MaterialFunctionCall("Animation  Phase (0-1)"=Divide(A=@n102, B=@n104) {constB=4}, Number of Rows=@n103.R, Number of Columns=@n103.R, UVs=@n107) {materialFunction=&FlipBook}.UVs
   Tex=NamedRerouteUsage()
  ) {samplerSource=SSM_Wrap_WorldGroupSettings, texture=&T_PipBoyEmotes}
  n109 = MaterialFunctionCall(UV=@n107) {materialFunction=&MF_UVMask01Space}
  NamedRerouteDeclaration(None=Multiply(A=Multiply(A=@n108.RGB, B=@n109.Result), B=NamedRerouteUsage())) {name=HealingEmote, nodeColor=<rgba 1 0 0.5391 1>}
  n110 = Constant()
  n111 = Multiply(A=Time(), B=ScalarParameter() {defaultValue=-0.6, group=Emote, parameterName=DamageDistortBreakSpeed})
  n112 = TextureSampleParameter2D(UVs=Add(A=NamedRerouteUsage(), B=AppendVector(A=@n110, B=@n111))) {
   parameterName=BreakLines
   samplerType=SAMPLERTYPE_Masks
   texture=/BRCosmetics/Accessories/FORT_Backpacks/Backpack_M_MED_RadiumFox/Textures/T_ScanBrokes
  }
  n113 = Subtract(A=@n112.R) {constB=0.5}
  n114 = NamedRerouteUsage()
  n115 = Multiply(A=Time(), B=ScalarParameter() {defaultValue=-0.4, group=Emote, parameterName=DamageDistortSpeed})
  n116 = Add(A=Multiply(A=ComponentMask(None=@n114) {g=true}) {constB=1.5}, B=@n115)
  n117 = SmoothStep(Value=Power(Base=Frac(None=@n116)) {constExponent=6}) {constMin=0.18}
  n118 = Saturate(None=Sign(None=@n117))
  n119 = Multiply(A=Multiply(A=Multiply(A=@n113) {constB=2}, B=ScalarParameter() {defaultValue=0.02, parameterName=DamageDistortBreakIntensity}), B=@n118)
  n120 = AppendVector(A=@n119, B=@n119)
  n121 = Multiply(A=Multiply(A=@n118, B=Sine(None=@n117)), B=ScalarParameter() {defaultValue=0.012, parameterName=DamageDistortIntensity})
  n122 = Constant()
  NamedRerouteDeclaration(None=Add(A=@n120, B=AppendVector(A=@n121, B=@n122))) {name=DamageDistort, nodeColor=<rgba 0 0.4844 1 1>}
  NamedRerouteDeclaration(None=@n1.BaseColor) {name=BaseColor_IN, nodeColor=<rgba 0.125 0 1 1>}
  NamedRerouteDeclaration(None=@n115) {name=DistortY, nodeColor=<rgba 0 0.5078 1 1>}
  NamedRerouteDeclaration(None=@n120) {name=BreakDistort, nodeColor=<rgba 0.5 0 1 1>}
  NamedRerouteDeclaration(None=@n112.R) {name=BreakMask, nodeColor=<rgba 0.1719 0 1 1>}
  NamedRerouteDeclaration(None=@n118) {name=ImpactMask, nodeColor=<rgba 0 0.7422 1 1>}
  n123 = FunctionInput() {bUsePreviewValueAsDefault=true, inputName=TechArtMask, sortPriority=1, previewValue=<1 1 1 1>}
  NamedRerouteDeclaration(None=ChannelMaskParameter(None=@n123) {group=Reactivity, parameterName=UseForScreen}.R) {name=TechArtMask, nodeColor=<rgba 0 1 0.9766 1>}
  n124 = TextureObjectParameter() {
   group=Reactivity
   parameterName=NumberAtlas
   texture=&T_FallOutNumbers
  }
  NamedRerouteDeclaration(None=@n124) {name=NumberTex, nodeColor=<rgba 0 1 0.4844 1>}
  n125 = Constant()
  NamedRerouteDeclaration(None=If(A=NamedRerouteUsage(), "A > B"=@n125, "A == B"=Constant() {r=1}, "A < B"=@n125)) {name=IsWalking, nodeColor=<rgba 0.4531 1 0 1>}
  n126 = Multiply(A=ScalarParameter() {defaultValue=1, group=Emote, parameterName=WalkingEmoteSpeed}, B=Time())
  n127 = Floor(None=LinearInterpolate(Alpha=Frac(None=@n126)) {constB=4})
  n128 = Constant2Vector() {g=2, r=2}
  n129 = Multiply(A=@n128.R, B=@n128.R)
  n130 = VectorParameter() {group=Reactivity, parameterName=Walking_Counter, channelNames {a=ScaleY, b=ScaleX, g=CenterY, r=CenterX}, defaultValue=<rgba 0.5 0.5 1 1>}
  MaterialExpressionAppendVector_5 = AppendVector(A=@n130.CenterX, B=@n130.CenterX)
  MaterialExpressionAppendVector_6 = AppendVector(A=@n130.ScaleX, B=@n130.ScaleX)
  MaterialExpressionNamedRerouteUsage_4 = NamedRerouteUsage()
  n131 = Custom(UVPoint=@MaterialExpressionAppendVector_5, UVScale=@MaterialExpressionAppendVector_6, OriginalUV=@MaterialExpressionNamedRerouteUsage_4) {
   code="float2 DistUV =OriginalUV-UVPoint;\r\nDistUV = DistUV/UVScale + float2(0.5,0.5);\r\n\r\nreturn DistUV;"
   description=CopyToPointUV
   outputType=CMOT_Float2
   inputs [
    {
     inputName=UVPoint
     input {expression=&self:MaterialExpressionAppendVector_5}
    }
    {
     inputName=UVScale
     input {expression=&self:MaterialExpressionAppendVector_6}
    }
    {
     inputName=OriginalUV
     input {expression=&self:MaterialExpressionNamedRerouteUsage_4}
    }
   ]
  }
  n132 = TextureSampleParameter2D(UVs=MaterialFunctionCall("Animation  Phase (0-1)"=Divide(A=@n127, B=@n129) {constB=4}, Number of Rows=@n128.R, Number of Columns=@n128.R, UVs=@n131) {materialFunction=&FlipBook}.UVs) {parameterName=WalkingIdle, texture=/BRCosmetics/Accessories/FORT_Backpacks/Backpack_M_MED_RadiumFox/Textures/T_RadiumFoxWalkingIdle}
  n133 = MaterialFunctionCall(UV=@n131) {materialFunction=&MF_UVMask01Space}
  NamedRerouteDeclaration(None=Multiply(A=NamedRerouteUsage(), B=Multiply(A=@n132.RGB, B=@n133.Result))) {name=WalkingEmote, nodeColor=<rgba 1 0 0.7734 1>}
  NamedRerouteUsage()
  NamedRerouteUsage()
 ]}
